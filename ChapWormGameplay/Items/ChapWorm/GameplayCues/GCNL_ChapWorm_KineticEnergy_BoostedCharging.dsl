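; Looping gameplay cue for the ChapWorm item's boosted kinetic energy charging, driving a material effect on the player pawn.
; Each event calls its parent, then finds the skeletal mesh of the pawn's "Gameplay" custom character part.
; Custom primitive data carries the state: application and removal stamp game time for an impulse, while loop start and removal toggle a steady glow.
; Loop start also prints a debug "Perma" string.
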(event EventOnApplication(EitherVFX-preferthisone)
  (params (MyTarget :type Actor%20Object%20Reference) (Parameters :type Gameplay%20Cue%20Parameters%20Structure%20%28by%20ref%29) (ParticleComponents :type Array%20of%20FXSystem%20Component%20Object%20References) (AudioComponents :type Array%20of%20Audio%20Component%20Object%20References) (BurstCameraShakeInstance :type Camera%20Shake%20Base%20Object%20Reference) (BurstDecalInstance :type Decal%20Actor%20Object%20Reference))
  ; Impulse Logic
  (|CallParentFunction MyTarget Parameters ParticleComponents AudioComponents BurstCameraShakeInstance BurstDecalInstance)
  (bind _asfort_player_pawn (Utilities|Casting|CastToFortPlayerPawn MyTarget))
  (Utilities|IsValid (FortPawn|GetModifierforPartType _asfort_player_pawn "Gameplay")
    (:Is%20Valid
      (Utilities|IsValid (CustomCharacter|GetSkeletalMeshForPartOnPlayerPawn (FortPawn|GetModifierforPartType _asfort_player_pawn "Gameplay"))
        (:Is%20Valid
          (Rendering|Material|SetCustomPrimitiveDataFloat (CustomCharacter|GetSkeletalMeshForPartOnPlayerPawn (FortPawn|GetModifierforPartType _asfort_player_pawn "Gameplay")) 12 (Utilities|Time|GetGameTimeinSeconds _asfort_player_pawn)))))))

(event EventOnLoopingStart(EitherVFX-preferthisversion)
  (params (MyTarget :type Actor%20Object%20Reference) (Parameters :type Gameplay%20Cue%20Parameters%20Structure%20%28by%20ref%29) (ParticleComponents :type Array%20of%20FXSystem%20Component%20Object%20References) (AudioComponents :type Array%20of%20Audio%20Component%20Object%20References))
  ; Permenant Glow logic
  (|CallParentFunction MyTarget Parameters ParticleComponents AudioComponents)
  (bind _asfort_player_pawn (Utilities|Casting|CastToFortPlayerPawn MyTarget))
  (Utilities|IsValid (FortPawn|GetModifierforPartType _asfort_player_pawn "Gameplay")
    (:Is%20Valid
      (Utilities|IsValid (CustomCharacter|GetSkeletalMeshForPartOnPlayerPawn (FortPawn|GetModifierforPartType _asfort_player_pawn "Gameplay"))
        (:Is%20Valid
          (Rendering|Material|SetCustomPrimitiveDataFloat (CustomCharacter|GetSkeletalMeshForPartOnPlayerPawn (FortPawn|GetModifierforPartType _asfort_player_pawn "Gameplay")) 13 1.0)
          (Development|PrintString 0 "Perma"))))))

(event EventOnRemoval(EitherVFX-preferthisversion)
  (params (MyTarget :type Actor%20Object%20Reference) (Parameters :type Gameplay%20Cue%20Parameters%20Structure%20%28by%20ref%29) (ParticleComponents :type Array%20of%20FXSystem%20Component%20Object%20References) (AudioComponents :type Array%20of%20Audio%20Component%20Object%20References) (BurstCameraShakeInstance :type Camera%20Shake%20Base%20Object%20Reference) (BurstDecalInstance :type Decal%20Actor%20Object%20Reference))
  ; Permenant Glow logic
  (|CallParentFunction MyTarget Parameters ParticleComponents AudioComponents BurstCameraShakeInstance BurstDecalInstance)
  (bind _asfort_player_pawn (Utilities|Casting|CastToFortPlayerPawn MyTarget))
  (Utilities|IsValid (FortPawn|GetModifierforPartType _asfort_player_pawn "Gameplay")
    (:Is%20Valid
      (Utilities|IsValid (CustomCharacter|GetSkeletalMeshForPartOnPlayerPawn (FortPawn|GetModifierforPartType _asfort_player_pawn "Gameplay"))
        (:Is%20Valid
          (Rendering|Material|SetCustomPrimitiveDataFloat (CustomCharacter|GetSkeletalMeshForPartOnPlayerPawn (FortPawn|GetModifierforPartType _asfort_player_pawn "Gameplay")) 13 0.0)
          (Rendering|Material|SetCustomPrimitiveDataFloat (CustomCharacter|GetSkeletalMeshForPartOnPlayerPawn (FortPawn|GetModifierforPartType _asfort_player_pawn "Gameplay")) 12 (Utilities|Time|GetGameTimeinSeconds _asfort_player_pawn)))))))

(fn ConstructionScript
  (|CallParentFunction))
