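; Looping gameplay cue for the ChapWorm item's maxed kinetic energy state, driving claw material and particle effects on the player pawn.
; On application, it stamps game time into the gameplay part mesh's custom primitive data, waiting a tick if that mesh is not ready.
; Water enter/exit events, bound on loop start and unbound on removal, toggle particle water parameters, with a timer polling immersion depth after entry.
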(event EventOnApplication(EitherVFX-preferthisone)
  (params (MyTarget :type Actor%20Object%20Reference) (Parameters :type Gameplay%20Cue%20Parameters%20Structure%20%28by%20ref%29) (ParticleComponents :type Array%20of%20FXSystem%20Component%20Object%20References) (AudioComponents :type Array%20of%20Audio%20Component%20Object%20References) (BurstCameraShakeInstance :type Camera%20Shake%20Base%20Object%20Reference) (BurstDecalInstance :type Decal%20Actor%20Object%20Reference))
  ; OnApplication: Update the material on the claws.
  ; Since this gameplay cue can be invoked before the character parts are ready, set up attempts to try again in the next frame.
  (|CallParentFunction MyTarget Parameters ParticleComponents AudioComponents BurstCameraShakeInstance BurstDecalInstance)
  (bind _asfort_player_pawn (Utilities|Casting|CastToFortPlayerPawn MyTarget))
  (Utilities|IsValid (FortPawn|GetModifierforPartType _asfort_player_pawn "Gameplay")
    (:Is%20Valid
      (Utilities|IsValid (CustomCharacter|GetSkeletalMeshForPartOnPlayerPawn (FortPawn|GetModifierforPartType _asfort_player_pawn "Gameplay"))
        (:Is%20Valid
          _)
        (:Is%20Not%20Valid
          (Utilities|FlowControl|DelayUntilNextTick _asfort_player_pawn)))
      (Rendering|Material|SetCustomPrimitiveDataFloat (CustomCharacter|GetSkeletalMeshForPartOnPlayerPawn (FortPawn|GetModifierforPartType _asfort_player_pawn "Gameplay")) 11 (Utilities|Time|GetGameTimeinSeconds _asfort_player_pawn)))
    (:Is%20Not%20Valid
      _)))

(event EventOnLoopingStart(EitherVFX-preferthisversion)
  (params (MyTarget :type Actor%20Object%20Reference) (Parameters :type Gameplay%20Cue%20Parameters%20Structure%20%28by%20ref%29) (ParticleComponents :type Array%20of%20FXSystem%20Component%20Object%20References) (AudioComponents :type Array%20of%20Audio%20Component%20Object%20References))
  (bind _asfort_player_pawn (Utilities|Casting|CastToFortPlayerPawn MyTarget))
  (Variables|Default|SetPlayerPawn _asfort_player_pawn)
  (Utilities|FlowControl|Sequence
    (:then_1
      (EventDispatchers|BindEventtoOnPawnExitWater (Variables|Default|GetPlayerPawn) (EventDispatchers|CreateEvent :!event Custom|Exit%20Water))
      (EventDispatchers|BindEventtoOnPawnEnterWater (Variables|Default|GetPlayerPawn) (EventDispatchers|CreateEvent :!event Custom|Enter%20Water)))))

(event Custom|Enter%20Water
  (params (FortPlayerPawn :type Fort%20Player%20Pawn%20Object%20Reference))
  (Effects|Components|ParticleSystem|SetFloatParameter (Utilities|Array|Get(acopy) (stale EventOnLoopingStart(EitherVFX-preferthisversion) ParticleComponents)) "On_Water" 0.0)
  (Variables|Default|SetLoop true)
  (Utilities|Time|SetTimerbyEvent (delegate Custom|Looping%20in%20water) 0.5 (Variables|Default|GetLoop)))

(event Custom|Looping%20in%20water
  (Effects|Components|ParticleSystem|SetFloatParameter (Utilities|Array|Get(acopy) (stale EventOnLoopingStart(EitherVFX-preferthisversion) ParticleComponents)) "In_Water" (Math|Conversions|ToFloat(Boolean) (> (WaterBody|GetWaterImmersionDepth (Variables|Default|GetPlayerPawn)) 8e+01))))

(event Custom|Exit%20Water
  (params (FortPlayerPawn :type Fort%20Player%20Pawn%20Object%20Reference) (bPawnIsOutsideOfAllWaterBodies :type Boolean))
  (Variables|Default|SetLoop false)
  (Effects|Components|ParticleSystem|SetFloatParameter (Utilities|Array|Get(acopy) (stale EventOnLoopingStart(EitherVFX-preferthisversion) ParticleComponents)) "On_Water" 1.0)
  (Effects|Components|ParticleSystem|SetFloatParameter (Utilities|Array|Get(acopy) (stale EventOnLoopingStart(EitherVFX-preferthisversion) ParticleComponents)) "In_Water" 0.0))

(event EventOnRemoval(EitherVFX-preferthisversion)
  (params (MyTarget :type Actor%20Object%20Reference) (Parameters :type Gameplay%20Cue%20Parameters%20Structure%20%28by%20ref%29) (ParticleComponents :type Array%20of%20FXSystem%20Component%20Object%20References) (AudioComponents :type Array%20of%20Audio%20Component%20Object%20References) (BurstCameraShakeInstance :type Camera%20Shake%20Base%20Object%20Reference) (BurstDecalInstance :type Decal%20Actor%20Object%20Reference))
  ; Event On Removal
  (EventDispatchers|UnbindEventfromOnPawnExitWater (Variables|Default|GetPlayerPawn) (EventDispatchers|CreateEvent :!event Custom|Exit%20Water))
  (EventDispatchers|UnbindEventfromOnPawnEnterWater (Variables|Default|GetPlayerPawn) (EventDispatchers|CreateEvent :!event Custom|Enter%20Water)))

(fn InitializeFX
  (if (> (WaterBody|GetWaterImmersionDepth (Variables|Default|GetPlayerPawn)) 0.1)
    (CallFunction|EnterWater (Variables|Default|GetPlayerPawn))
    (else
      (CallFunction|ExitWater (Variables|Default|GetPlayerPawn)))))

(fn ConstructionScript)
