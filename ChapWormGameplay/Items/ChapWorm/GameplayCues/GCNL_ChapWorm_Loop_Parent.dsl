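; Parent looping GameplayCue notify for the ChapWorm item, adding controller feedback and a radial force effect when the cue is applied.
; On application it calls the parent, then plays force feedback on the target's player controller, picking the Overpower variant when the target has the ChapWorm.Overpower tag.
; If enabled and the platform does not prefer performance, it spawns a radial force field actor at the target after a one-tick delay.
; A camera shake helper is also defined.
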(event EventOnApplication(EitherVFX-preferthisone)
  (params (MyTarget :type Actor%20Object%20Reference) (Parameters :type Gameplay%20Cue%20Parameters%20Structure%20%28by%20ref%29) (ParticleComponents :type Array%20of%20FXSystem%20Component%20Object%20References) (AudioComponents :type Array%20of%20Audio%20Component%20Object%20References) (BurstCameraShakeInstance :type Camera%20Shake%20Base%20Object%20Reference) (BurstDecalInstance :type Decal%20Actor%20Object%20Reference))
  ;[0.01704 0.1302 0.04533] OnApplication: Apply radial force and force feedback.
  (|CallParentFunction MyTarget Parameters ParticleComponents AudioComponents BurstCameraShakeInstance BurstDecalInstance)
  (Utilities|IsValid MyTarget
    (:Is%20Valid
      (bind _returnvalue (GameplayTags|HasMatchingGameplayTag(Message) MyTarget "(TagName=\"ChapWorm.Overpower\")"))
      (Default|PlayForceFeedback MyTarget (select _returnvalue (Variables|ForceFeedback|GetLocalPlayerForceFeedbackEffect_Overpower) (Variables|ForceFeedback|GetLocalPlayerForceFeedbackEffect_Default)))
      (if (and (not (Default|PreferPerformance)) (Variables|RadialForce|GetbSpawnRadialForce))
        (Utilities|FlowControl|DelayUntilNextTick MyTarget) ; Tick delay to help perf
        (Game|SpawnActorfromClass MyTarget "/Game/Effects/Niagara/Blueprints/PhsyicsFields/FN_RadialForce.FN_RadialForce_C" (Math|Conversions|ToTransform(Vector) (Transformation|GetActorLocation MyTarget)) "Undefined" "MultiplyWithRoot" 0 (Variables|RadialForce|GetRadialForcePlayRate) (Variables|RadialForce|GetRadialForceTimeDecay) (Variables|RadialForce|GetRadialForceImpactRadius) (Variables|RadialForce|GetRadialForceStrength) (Variables|RadialForce|GetRadialForceDelayTime) 0)))))

(fn PlayCameraShake
  (params (TargetActor :type Actor%20Object%20Reference) (ShakeClass :type Camera%20Shake%20Base%20Class%20Reference) (Scale :type Float%20%28single-precision%29))
  (bind _returnvalue (Fortnite|GetFortPlayerControllerfromActor TargetActor))
  ; Play the desired camera shake on the owning player controller.
  (Utilities|IsValid _returnvalue
    (:Is%20Valid
      (bind _playercameramanager (Variables|/Script/Engine.PlayerController|GetPlayerCameraManager _returnvalue)
        (:then
          (CameraShakes|StartCameraShake :self _playercameramanager :ShakeClass ShakeClass :Scale Scale :PlaySpace "CameraLocal"))
        (:else)))))

(fn PlayForceFeedback
  (params (TargetActor :type Actor%20Object%20Reference) (ForceFeedbackEffect :type Force%20Feedback%20Effect%20Object%20Reference))
  (bind _returnvalue (Fortnite|GetFortPlayerControllerfromActor TargetActor))
  ;[0.008391 0.1308 0.1406] Play force feedback on the target actor (the player that is instigating the attack)
  (Utilities|IsValid TargetActor
    (:Is%20Valid
      (Utilities|IsValid _returnvalue
        (:Is%20Valid
          (Game|Feedback|LocalPlayForceFeedback _returnvalue ForceFeedbackEffect false "None"))))))

(fn PreferPerformance
  ;[1 0.3356 0.7917] Does the system prefer performance?
  (return (Math|Boolean|ORBoolean (Fortnite|IsMobilePlatform) (Class|PlatformHelpersLibrary|IsPcPerfMode self) (TouchPlatform|IsMobileGame self) (TouchPlatform|IsNintendoSwitchorNintendoSwitchPreview))))

(fn ConstructionScript)
