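; Looping gameplay cue for the ChapWorm Overpower effect, which updates the claw material on the player pawn.
; On application it writes game time and an on flag to custom primitive data on the "Gameplay" character part mesh, delaying a tick if that mesh is not ready.
; It also sets burst camera shake scale based on whether the current weapon is being equipped.
; Removal writes the time again and clears the flag.
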(event EventOnApplication(EitherVFX-preferthisone)
  (params (MyTarget :type Actor%20Object%20Reference) (Parameters :type Gameplay%20Cue%20Parameters%20Structure%20%28by%20ref%29) (ParticleComponents :type Array%20of%20FXSystem%20Component%20Object%20References) (AudioComponents :type Array%20of%20Audio%20Component%20Object%20References) (BurstCameraShakeInstance :type Camera%20Shake%20Base%20Object%20Reference) (BurstDecalInstance :type Decal%20Actor%20Object%20Reference))
  ; OnApplication: Update the material on the claws.
  ; Since this gameplay cue can be invoked before the character parts are ready, set up attempts to try again in the next frame.
  (|CallParentFunction MyTarget Parameters ParticleComponents AudioComponents BurstCameraShakeInstance BurstDecalInstance)
  (bind _asfort_player_pawn (Utilities|Casting|CastToFortPlayerPawn MyTarget))
  (if (Class|FortWeapon|IsBeingEquipped (Variables|/Script/FortniteGame.FortPawn|GetCurrentWeapon _asfort_player_pawn))
    (Variables|/Script/Engine.CameraShakeBase|SetShakeScale 0.0 BurstCameraShakeInstance)
    (Utilities|IsValid (FortPawn|GetModifierforPartType _asfort_player_pawn "Gameplay")
      (:Is%20Valid
        (Utilities|IsValid (CustomCharacter|GetSkeletalMeshForPartOnPlayerPawn (FortPawn|GetModifierforPartType _asfort_player_pawn "Gameplay"))
          (:Is%20Valid
            _)
          (:Is%20Not%20Valid
            (Utilities|FlowControl|DelayUntilNextTick _asfort_player_pawn)))
        (Rendering|Material|SetCustomPrimitiveDataFloat (CustomCharacter|GetSkeletalMeshForPartOnPlayerPawn (FortPawn|GetModifierforPartType _asfort_player_pawn "Gameplay")) 14 (Utilities|Time|GetGameTimeinSeconds _asfort_player_pawn))
        (Rendering|Material|SetCustomPrimitiveDataFloat (CustomCharacter|GetSkeletalMeshForPartOnPlayerPawn (FortPawn|GetModifierforPartType _asfort_player_pawn "Gameplay")) 15 1.0))
      (:Is%20Not%20Valid
        _)))
  (Variables|/Script/Engine.CameraShakeBase|SetShakeScale 1.0 BurstCameraShakeInstance))

(event EventOnRemoval(EitherVFX-preferthisversion)
  (params (MyTarget :type Actor%20Object%20Reference) (Parameters :type Gameplay%20Cue%20Parameters%20Structure%20%28by%20ref%29) (ParticleComponents :type Array%20of%20FXSystem%20Component%20Object%20References) (AudioComponents :type Array%20of%20Audio%20Component%20Object%20References) (BurstCameraShakeInstance :type Camera%20Shake%20Base%20Object%20Reference) (BurstDecalInstance :type Decal%20Actor%20Object%20Reference))
  (|CallParentFunction MyTarget Parameters ParticleComponents AudioComponents BurstCameraShakeInstance BurstDecalInstance)
  (bind _asfort_player_pawn (Utilities|Casting|CastToFortPlayerPawn MyTarget))
  (Utilities|IsValid (FortPawn|GetModifierforPartType _asfort_player_pawn "Gameplay")
    (:Is%20Valid
      (Utilities|IsValid (CustomCharacter|GetSkeletalMeshForPartOnPlayerPawn (FortPawn|GetModifierforPartType _asfort_player_pawn "Gameplay"))
        (:Is%20Valid
          (Rendering|Material|SetCustomPrimitiveDataFloat (CustomCharacter|GetSkeletalMeshForPartOnPlayerPawn (FortPawn|GetModifierforPartType _asfort_player_pawn "Gameplay")) 14 (Utilities|Time|GetGameTimeinSeconds _asfort_player_pawn))
          (Rendering|Material|SetCustomPrimitiveDataFloat (CustomCharacter|GetSkeletalMeshForPartOnPlayerPawn (FortPawn|GetModifierforPartType _asfort_player_pawn "Gameplay")) 15 0.0))))))

(fn ConstructionScript)
