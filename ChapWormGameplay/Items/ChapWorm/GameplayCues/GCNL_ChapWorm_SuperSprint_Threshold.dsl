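; Looping gameplay cue for the ChapWorm super sprint threshold that adds sonic boom audio and a camera shake when applied.
; On application it runs the parent logic, then casts the target to a Fort player pawn.
; It attaches a sound to the pawn's mesh, choosing the overpower or base sonic boom cue by the ChapWorm.Overpower tag.
; It then plays the dash charge camera shake on that pawn.
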(event EventOnApplication(EitherVFX-preferthisone)
  (params (MyTarget :type Actor%20Object%20Reference) (Parameters :type Gameplay%20Cue%20Parameters%20Structure%20%28by%20ref%29) (ParticleComponents :type Array%20of%20FXSystem%20Component%20Object%20References) (AudioComponents :type Array%20of%20Audio%20Component%20Object%20References) (BurstCameraShakeInstance :type Camera%20Shake%20Base%20Object%20Reference) (BurstDecalInstance :type Decal%20Actor%20Object%20Reference))
  (|CallParentFunction MyTarget Parameters ParticleComponents AudioComponents BurstCameraShakeInstance BurstDecalInstance)
  (bind _asfort_player_pawn (Utilities|Casting|CastToFortPlayerPawn MyTarget))
  (Audio|SpawnSoundAttached (select (GameplayTags|HasMatchingGameplayTag _asfort_player_pawn "(TagName=\"ChapWorm.Overpower\")") (Variables|Default|GetOverpowerSonicBoomSoundCue) (Variables|Default|GetBaseSonicBoomSoundCue)) (Variables|/Script/Engine.Character|GetMesh _asfort_player_pawn))
  (Default|PlayCameraShake _asfort_player_pawn "/ChapWormGameplay/Items/ChapWorm/FX/Misc/B_ChapWorm_Dash_Charge_On.B_ChapWorm_Dash_Charge_On_C" 1.0))

(fn ConstructionScript)
