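; Looping gameplay cue notify that plays a local camera shake during ChapWorm's super sprint.
; On loop start it starts the shake on the player camera manager at base scale and stores the instance.
; Async tag-add and tag-remove listeners on the pawn switch that live shake between Overpower and base scale, without restarting it.
; When the cue ceases relevance, it stops the shake non-immediately and ends only the tag-added listener.
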(event EventOnLoopingStart(EitherVFX-preferthisversion)
  (params (MyTarget :type Actor%20Object%20Reference) (Parameters :type Gameplay%20Cue%20Parameters%20Structure%20%28by%20ref%29) (ParticleComponents :type Array%20of%20FXSystem%20Component%20Object%20References) (AudioComponents :type Array%20of%20Audio%20Component%20Object%20References))
  ;[0.03099 0.1094 0.0343] OnLoopingStart: Apply screenshake and adjust scale based on overpower state.
  (|CallParentFunction MyTarget Parameters ParticleComponents AudioComponents)
  (bind _fortplayercontroller (Variables|Default|GetFortPlayerController)
    (:then
      (bind _playercameramanager (Variables|/Script/Engine.PlayerController|GetPlayerCameraManager _fortplayercontroller)
        (:then
          (bind _returnvalue (CameraShakes|StartCameraShake :self _playercameramanager :ShakeClass (Variables|Default|GetSuperSprintCameraShakeClass) :Scale (Variables|Default|GetBaseCameraShakeScale) :PlaySpace "CameraLocal"))
          (Variables|Transient|SetSuperSprintCameraShake _returnvalue)
          (bind _asynctaskproxy (Ability|Tasks|WaitGameplayTagAddToActor (Game|Player|GetPlayerPawn _fortplayercontroller) (Variables|Default|GetOverpowerGameplayTag))
            (:then
              (Variables|Transient|SetAsyncAction_OverpowerAdded _asynctaskproxy)
              (bind _asynctaskproxy_2 (Ability|Tasks|WaitGameplayTagRemoveFromActor (Game|Player|GetPlayerPawn _fortplayercontroller) (Variables|Default|GetOverpowerGameplayTag))
                (:then
                  (Variables|Transient|SetAsyncAction_OverpowerRemove _asynctaskproxy_2))
                (:Removed
                  (bind _supersprintcamerashake (Variables|Transient|GetSuperSprintCameraShake)
                    (:then
                      (Variables|/Script/Engine.CameraShakeBase|SetShakeScale (Variables|Default|GetBaseCameraShakeScale) _supersprintcamerashake))
                    (:else)))))
            (:Added
              (bind _supersprintcamerashake_2 (Variables|Transient|GetSuperSprintCameraShake)
                (:then
                  (Variables|/Script/Engine.CameraShakeBase|SetShakeScale (Variables|Default|GetOverpowerCameraShakeScale) _supersprintcamerashake_2))
                (:else)))))
        (:else)))
    (:else)))

(fn OnCeaseRelevant
  (params (MyTarget :type Actor%20Object%20Reference) (Parameters :type Gameplay%20Cue%20Parameters%20Structure%20%28by%20ref%29))
  ;[0.1406 0.07196 0.00617] OnRemove: Clean up screenshake and async actions.
  (|CallParentFunction MyTarget Parameters)
  (Utilities|FlowControl|Sequence
    (:then_0
      (Ability|Async|EndAction (Variables|Transient|GetAsyncAction_OverpowerAdded))
      (bind _supersprintcamerashake (Variables|Transient|GetSuperSprintCameraShake)
        (:then
          (bind _fortplayercontroller (Variables|Default|GetFortPlayerController)
            (:then
              (bind _playercameramanager (Variables|/Script/Engine.PlayerController|GetPlayerCameraManager _fortplayercontroller)
                (:then
                  (CameraShakes|StopCameraShake _playercameramanager _supersprintcamerashake false))
                (:else)))
            (:else)))
        (:else)))
    (:then_1
      (return false))))

(fn ConstructionScript
  (|CallParentFunction))
