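; Parent burst GameplayCue notify for the ChapWorm item, layering shared feedback and physics effects on the inherited burst.
; OnBurst plays local force feedback chosen by the ChapWorm.Overpower tag, shows an optional team sound indicator, and passes the hit surface type to the particle component.
; When enabled and outside mobile, Switch or PC performance mode, it spawns a radial force field one tick later to help performance.
; A camera shake helper is defined but not called here.
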(event EventOnBurst(EitherVFX-preferthisversion)
  (params (MyTarget :type Actor%20Object%20Reference) (Parameters :type Gameplay%20Cue%20Parameters%20Structure%20%28by%20ref%29) (ParticleComponents :type Array%20of%20FXSystem%20Component%20Object%20References) (AudioComponents :type Array%20of%20Audio%20Component%20Object%20References) (BurstCameraShakeInstance :type Camera%20Shake%20Base%20Object%20Reference) (BurstDecalInstance :type Decal%20Actor%20Object%20Reference))
  (bind (_normalizedmagnitude _rawmagnitude _effectcontext) (Ability|GameplayCue|BreakGameplayCueParameters Parameters))
  (bind _output (Utilities|Array|Get(acopy) ParticleComponents))
  ;[0.05759 0.2188 0.0669] OnBurst: Apply radial force and force feedback.
  (|CallParentFunction MyTarget Parameters ParticleComponents AudioComponents BurstCameraShakeInstance BurstDecalInstance)
  (Utilities|IsValid MyTarget
    (:Is%20Valid
      (bind _returnvalue (GameplayTags|HasMatchingGameplayTag(Message) MyTarget "(TagName=\"ChapWorm.Overpower\")"))
      (Default|PlayForceFeedback MyTarget (select _returnvalue (Variables|ForceFeedback|GetLocalPlayerForceFeedbackEffect_Overpower) (Variables|ForceFeedback|GetLocalPlayerForceFeedbackEffect_Default)))
      (CallFunction|SpawnSoundIndicator MyTarget AudioComponents)
      (if (Utilities|Array|IsValidIndex ParticleComponents)
        (Utilities|IsValid _output
          (:Is%20Valid
            (Effects|Components|ParticleSystem|SetIntParameter _output "PhysicalSurface" (Utilities|Select :Index (Collision|GetFortPhysicalSurface (Ability|EffectContext|GetHitResult _effectcontext)) :SurfaceType_Default 0 :SurfaceType1 1 :SurfaceType2 2 :SurfaceType3 3 :SurfaceType4 4 :SurfaceType5 5 :SurfaceType6 6 :SurfaceType7 7 :SurfaceType8 8 :SurfaceType9 9 :SurfaceType10 10 :SurfaceType11 11 :SurfaceType12 12 :SurfaceType13 13 :SurfaceType14 14 :SurfaceType15 15 :SurfaceType16 16 :SurfaceType17 17 :SurfaceType18 18 :SurfaceType19 19 :SurfaceType20 20 :SurfaceType21 21 :SurfaceType22 22 :SurfaceType23 23 :SurfaceType24 24 :SurfaceType25 25 :SurfaceType26 26 :SurfaceType27 0 :SurfaceType28 0 :SurfaceType29 0 :SurfaceType30 0 :SurfaceType31 0 :SurfaceType32 0 :SurfaceType33 0 :SurfaceType34 0)))
          (:Is%20Not%20Valid
            _))
        (if (and (not (CallFunction|PreferPerformance)) (Variables|RadialForce|GetbSpawnRadialForce))
          (Utilities|FlowControl|DelayUntilNextTick MyTarget) ; Tick delay to help perf
          (Game|SpawnActorfromClass MyTarget "/Game/Effects/Niagara/Blueprints/PhsyicsFields/FN_RadialForce.FN_RadialForce_C" (Math|Conversions|ToTransform(Vector) (Transformation|GetActorLocation MyTarget)) "Undefined" "MultiplyWithRoot" 0 (Variables|RadialForce|GetRadialForcePlayRate) (Variables|RadialForce|GetRadialForceTimeDecay) (Variables|RadialForce|GetRadialForceImpactRadius) (Variables|RadialForce|GetRadialForceStrength) (Variables|RadialForce|GetRadialForceDelayTime) 0))
        (else
          _)))))

(fn PlayCameraShake
  (params (TargetActor :type Actor%20Object%20Reference) (ShakeClass :type Camera%20Shake%20Base%20Class%20Reference) (Scale :type Float%20%28single-precision%29))
  (bind _returnvalue (Fortnite|GetFortPlayerControllerfromActor TargetActor))
  ; Play the desired camera shake on the owning player controller.
  (Utilities|IsValid _returnvalue
    (:Is%20Valid
      (bind _playercameramanager (Variables|/Script/Engine.PlayerController|GetPlayerCameraManager _returnvalue)
        (:then
          (CameraShakes|StartCameraShake :self _playercameramanager :ShakeClass ShakeClass :Scale Scale :PlaySpace "CameraLocal"))
        (:else)))))

(fn PlayForceFeedback
  (params (TargetActor :type Actor%20Object%20Reference) (ForceFeedbackEffect :type Force%20Feedback%20Effect%20Object%20Reference))
  (bind _returnvalue (Fortnite|GetFortPlayerControllerfromActor TargetActor))
  ;[0.008391 0.1308 0.1406] Play force feedback on the target actor (the player that is instigating the attack)
  (Utilities|IsValid TargetActor
    (:Is%20Valid
      (Utilities|IsValid _returnvalue
        (:Is%20Valid
          (Game|Feedback|LocalPlayForceFeedback _returnvalue ForceFeedbackEffect false "None"))))))

(fn PreferPerformance
  ;[0.05618 0.2 0.2083] Does the system prefer performance?
  (return (Math|Boolean|ORBoolean (Fortnite|IsMobilePlatform) (Class|PlatformHelpersLibrary|IsPcPerfMode self) (TouchPlatform|IsMobileGame self) (TouchPlatform|IsNintendoSwitchorNintendoSwitchPreview))))

(fn SpawnSoundIndicator
  (params (TargetActor :type Actor%20Object%20Reference) (AudioComponents :type Array%20of%20Audio%20Component%20Object%20References))
  (bind _output (Utilities|Array|Get(acopy) AudioComponents))
  ;[0.02051 0.132 0.15 0.5] SpawnSoundIndicator: Spawn the sound indicator, if desired and there's a valid sound played.
  (if (Variables|SoundIndicator|GetbShowSoundIndicator)
    (if (Utilities|Array|IsValidIndex AudioComponents)
      (Utilities|IsValid _output
        (:Is%20Valid
          (bind _sound (Variables|/Script/Engine.AudioComponent|GetSound _output)
            (:then
              (Fortnite|ShowSoundIndicatorLocalforTeam :WorldContextObject TargetActor :Location (Transformation|GetActorLocation TargetActor) :MaxAudibleDistance (Variables|/Script/Engine.SoundBase|GetMaxDistance _sound) :IndicatorType "Gunshot" :Team (Team|GetActorTeam TargetActor) :Affiliations (Utilities|Array|MakeArray "Neutral" "Hostile") :Tint "(R=0.000000,G=0.000000,B=0.000000,A=0.000000)" :AttenuationMaxRangeMultiplier 1.0 :AttenuationIntensityMultiplier 1.0))
            (:else)))))))
