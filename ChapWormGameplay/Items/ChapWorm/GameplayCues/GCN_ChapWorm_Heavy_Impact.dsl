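; Burst gameplay cue for the ChapWorm heavy impact that adds a radial force effect at the target player pawn.
; OnBurst casts the target to a Fort player pawn and spawns an FN_RadialForce Niagara physics-field actor at its root component's location.
; The spawn is gated to non-mobile platforms outside PC performance mode with a rendering detail mode above 1.
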(fn OnBurst(EitherVFX-preferthisversion)
  (params (MyTarget :type Actor%20Object%20Reference) (Parameters :type Gameplay%20Cue%20Parameters%20Structure%20%28by%20ref%29) (ParticleComponents :type Array%20of%20FXSystem%20Component%20Object%20References) (AudioComponents :type Array%20of%20Audio%20Component%20Object%20References) (BurstCameraShakeInstance :type Camera%20Shake%20Base%20Object%20Reference) (BurstDecalInstance :type Decal%20Actor%20Object%20Reference))
  (Utilities|FlowControl|Sequence
    (:then_0
      (bind _asfort_player_pawn (Utilities|Casting|CastToFortPlayerPawn MyTarget))
      ;[1 0 0] VFX: Radial Force
      (if (Math|Boolean|ANDBoolean (not (Fortnite|IsMobilePlatform)) (not (Class|PlatformHelpersLibrary|IsPcPerfMode 0)) (> (Rendering|GetRenderingDetailMode) 1))
        (Game|SpawnActorfromClass 0 "/Game/Effects/Niagara/Blueprints/PhsyicsFields/FN_RadialForce.FN_RadialForce_C" (Math|Transform|MakeTransform (Transformation|GetWorldLocation (Variables|/Script/Engine.Actor|GetRootComponent _asfort_player_pawn))) "Undefined" "MultiplyWithRoot" 0 3.0 0.6 1.05e+03 0.9 0.1 0)))
    (:then_1
      (return))))
