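; Parent burst gameplay cue for ChapWorm melee hits, spawning impact effects and playing damage sounds.
; GetAdjustedImpactPoint line-traces the hit component from an offset on the target pawn toward its closest collision point, falling back to the raw impact.
; The burst spawns electric (ChapWorm.Overpower tag) or default Niagara there, passing physical surface and tuning variables.
; Sound Library plays flesh or generic damage audio by surface.
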(fn GetAdjustedImpactPoint
  (params (TargetActor :type Actor%20Object%20Reference) (Hit :type Hit%20Result%20Structure%20%28by%20ref%29))
  (bind (_bblockinghit _binitialoverlap _time _distance _location _impactpoint _normal _impactnormal _physmat _hitactor _hitcomponent) (Collision|BreakHitResult Hit))
  (bind _hitcomponent_2 (LocalVariables|GetHitComponent))
  ; AdjustedImpactPoint: Adjust the impact point so that it is closest to the bone/socket on the player that is actually doing the damage.
  (Utilities|FlowControl|Sequence
    (:then_0
      (bind _output_get (LocalVariables|SetHitComponent _hitcomponent))
      (Utilities|IsValid _output_get
        (:Is%20Valid
          (bind _asfort_player_pawn (Utilities|Casting|CastToFortPlayerPawn TargetActor))
          (bind _output_get_2 (LocalVariables|SetTraceStartLoc (Utilities|Operators|Add (* (.x (Variables|Default|GetHitTraceStartRelativeOffset)) (Transformation|GetActorForwardVector _asfort_player_pawn)) (* (Transformation|GetActorRightVector _asfort_player_pawn) (.y (Variables|Default|GetHitTraceStartRelativeOffset))) (* (Transformation|GetActorUpVector _asfort_player_pawn) (.z (Variables|Default|GetHitTraceStartRelativeOffset))) (Transformation|GetActorLocation _asfort_player_pawn))))
          (LocalVariables|SetClosestPointOnHitComponent (Collision|GetClosestPointonCollision _hitcomponent_2 _output_get_2 "None"))
          (if (>= (.ReturnValue (Collision|GetClosestPointonCollision _hitcomponent_2 _output_get_2 "None")) 0)
            (bind (_hitlocation _hitnormal _bonename _outhit _returnvalue) (Collision|LineTraceComponent (LocalVariables|GetHitComponent) (LocalVariables|GetTraceStartLoc) (+ (LocalVariables|GetTraceStartLoc) (* (Math|Vector|GetUnitDirection(Vector) (LocalVariables|GetTraceStartLoc) (LocalVariables|GetClosestPointOnHitComponent)) 1e+02)) true false false))
            (if _returnvalue
              (return _hitlocation _hitnormal))))))
    (:then_1
      (return _impactpoint _impactnormal))))

(fn OnBurst(EitherVFX-preferthisversion)
  (params (MyTarget :type Actor%20Object%20Reference) (Parameters :type Gameplay%20Cue%20Parameters%20Structure%20%28by%20ref%29) (ParticleComponents :type Array%20of%20FXSystem%20Component%20Object%20References) (AudioComponents :type Array%20of%20Audio%20Component%20Object%20References) (BurstCameraShakeInstance :type Camera%20Shake%20Base%20Object%20Reference) (BurstDecalInstance :type Decal%20Actor%20Object%20Reference))
  (bind (_normalizedmagnitude _rawmagnitude _effectcontext) (Ability|GameplayCue|BreakGameplayCueParameters Parameters))
  (bind _returnvalue (Ability|EffectContext|GetHitResult _effectcontext))
  (bind (_impactpoint _impactnormal) (CallFunction|GetAdjustedImpactPoint MyTarget _returnvalue))
  (bind _returnvalue_2 (Collision|GetFortPhysicalSurface _returnvalue))
  (bind _mytarget (Variables|GetMyTarget))
  (bind _returnvalue_3 (WorldSubsystems|GetSoundLibrarySubsystem _mytarget))
  (bind _returnvalue_4 (Utilities|Select :Index _returnvalue_2 :SurfaceType_Default 0 :SurfaceType1 1 :SurfaceType2 2 :SurfaceType3 3 :SurfaceType4 4 :SurfaceType5 5 :SurfaceType6 6 :SurfaceType7 7 :SurfaceType8 8 :SurfaceType9 9 :SurfaceType10 10 :SurfaceType11 11 :SurfaceType12 12 :SurfaceType13 13 :SurfaceType14 14 :SurfaceType15 15 :SurfaceType16 16 :SurfaceType17 17 :SurfaceType18 18 :SurfaceType19 19 :SurfaceType20 20 :SurfaceType21 21 :SurfaceType22 22 :SurfaceType23 23 :SurfaceType24 24 :SurfaceType25 25 :SurfaceType26 26 :SurfaceType27 0 :SurfaceType28 0 :SurfaceType29 0 :SurfaceType30 0 :SurfaceType31 0 :SurfaceType32 0 :SurfaceType33 0 :SurfaceType34 0))
  (|CallParentFunction MyTarget Parameters ParticleComponents AudioComponents BurstCameraShakeInstance BurstDecalInstance)
  (Utilities|FlowControl|Sequence
    (:then_0
      (Utilities|IsValid MyTarget
        (:Is%20Valid
          (bind _returnvalue_5 (GameplayTags|HasMatchingGameplayTag(Message) MyTarget "(TagName=\"ChapWorm.Overpower\")"))
          (bind _returnvalue_6 (Niagara|SpawnSystematLocation MyTarget (select _returnvalue_5 "/ChapWormGameplay/Items/ChapWorm/FX/NS_ChapWorm_Electric_Hit.NS_ChapWorm_Electric_Hit" "/ChapWormGameplay/Items/ChapWorm/FX/NS_ChapWorm_Melee_Hit_Default.NS_ChapWorm_Melee_Hit_Default") _impactpoint (Math|Conversions|RotationFromXVector _impactnormal) "1.000000,1.000000,1.000000" true true "AutoRelease"))
          (Utilities|IsValid _returnvalue_6
            (:Is%20Valid
              (Effects|Components|ParticleSystem|SetIntParameter _returnvalue_6 "PhysicalSurface" _returnvalue_4)
              (Niagara|SetNiagaraVariable(Vector3) _returnvalue_6 "Velocity" (Variables|Default|GetFX%20Velocity))
              (Niagara|SetNiagaraVariable(Bool) _returnvalue_6 "IS_Claw" (Variables|Default|GetIS_Claw))
              (Niagara|SetNiagaraVariable(Float) _returnvalue_6 "FX_Scale" (Variables|Default|GetFX%20Scale))
              (Niagara|SetNiagaraVariable(Vector3) _returnvalue_6 "Decal Scale Adjustment" (Variables|Default|GetDecal%20Scale%20Adjustment))
              (Niagara|SetNiagaraVariable(Vector3) _returnvalue_6 "RefVec Adjustment" (Variables|Default|GetRefVec%20Adjustment)))
            (:Is%20Not%20Valid
              _))
          (Utilities|IsValid _returnvalue_3
            (:Is%20Valid
              (SoundLibrary|PlaySoundSimple _returnvalue_3 _mytarget (Utilities|Struct|MakeSimpleSoundSettings (select (== _returnvalue_4 5) "(TagName=\"SoundLibrary.Melee.Damage.Flesh\")" "(TagName=\"SoundLibrary.Melee.Damage\")")) :!class /Script/SoundLibrary.SoundLibrarySubsystem))))))
    (:then_1
      (return))))
