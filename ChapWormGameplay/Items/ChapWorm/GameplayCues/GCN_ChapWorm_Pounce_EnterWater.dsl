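; Burst gameplay cue for the ChapWorm pounce entering water, spawning a splash Niagara effect on the water surface.
; OnBurst calls the parent, then casts the cue target to a Fort player pawn and gets its current water body.
; If that body is valid, it queries the surface at the pawn's location and spawns the configured splash system there, using auto-release pooling and pre-cull checks.
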(fn OnBurst(EitherVFX-preferthisversion)
  (params (MyTarget :type Actor%20Object%20Reference) (Parameters :type Gameplay%20Cue%20Parameters%20Structure%20%28by%20ref%29) (ParticleComponents :type Array%20of%20FXSystem%20Component%20Object%20References) (AudioComponents :type Array%20of%20Audio%20Component%20Object%20References) (BurstCameraShakeInstance :type Camera%20Shake%20Base%20Object%20Reference) (BurstDecalInstance :type Decal%20Actor%20Object%20Reference))
  ;[0.004079 0.1979 0.194] OnBurst: Apply the water splash niagara effect.
  (|CallParentFunction MyTarget Parameters ParticleComponents AudioComponents BurstCameraShakeInstance BurstDecalInstance)
  (Utilities|FlowControl|Sequence
    (:then_0
      (bind _watersplashniagarasystem (Variables|Default|GetWaterSplashNiagaraSystem)
        (:then
          (bind _asfort_player_pawn (Utilities|Casting|CastToFortPlayerPawn MyTarget))
          (Utilities|IsValid (WaterBody|GetCurrentWaterBody _asfort_player_pawn)
            (:Is%20Valid
              (bind _waterbodycomponent (Variables|/Script/Water.WaterBody|GetWaterBodyComponent (WaterBody|GetCurrentWaterBody _asfort_player_pawn))
                (:then
                  (Niagara|SpawnSystematLocation :WorldContextObject _waterbodycomponent :SystemTemplate _watersplashniagarasystem :Location (WaterBody|GetWaterSurfaceInfoatLocation _waterbodycomponent (Transformation|GetActorLocation _asfort_player_pawn) false) :Scale "1.000000,1.000000,1.000000" :bAutoDestroy true :bAutoActivate true :PoolingMethod "AutoRelease" :bPreCullCheck true))
                (:else)))))
        (:else)))
    (:then_1
      (return))))
