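; Parent burst gameplay cue for the ChapWorm pounce impact, adding camera and physics effects when the target player pawn is locally viewed.
; A swimming pawn gets a water lens effect.
; Otherwise, a "GroundPunch" montage section triggers a dirt lens effect and clears the particle system's IS_Fail flag.
; That ground-punch path also spawns a radial force field at the pawn, gated off mobile, PC performance mode and low detail.
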(fn OnBurst(EitherVFX-preferthisversion)
  (params (MyTarget :type Actor%20Object%20Reference) (Parameters :type Gameplay%20Cue%20Parameters%20Structure%20%28by%20ref%29) (ParticleComponents :type Array%20of%20FXSystem%20Component%20Object%20References) (AudioComponents :type Array%20of%20Audio%20Component%20Object%20References) (BurstCameraShakeInstance :type Camera%20Shake%20Base%20Object%20Reference) (BurstDecalInstance :type Decal%20Actor%20Object%20Reference))
  (bind _output (Utilities|Array|Get(acopy) ParticleComponents))
  ;[0.01174 0.05729 0.0333] OnBurst: Spawn camera effects.
  (Utilities|FlowControl|Sequence
    (:then_0
      (bind _asfort_player_pawn (Utilities|Casting|CastToFortPlayerPawn MyTarget))
      (if (Pawn|IsLocallyViewed _asfort_player_pawn)
        (if (GameplayTags|HasMatchingGameplayTag _asfort_player_pawn (Variables|Default|GetSwimmingTag))
          (Feedback|SpawnGenericCameraLensEffect _asfort_player_pawn (Variables|Default|GetCameraLensEffectWaterClass))
          (else
            (bind _mesh (Variables|/Script/Engine.Character|GetMesh _asfort_player_pawn)
              (:then
                (Utilities|IsValid (Components|SkeletalMesh|GetAnimInstance _mesh)
                  (:Is%20Valid
                    (if (== (Animation|Montage|MontageGetCurrentSection (Components|SkeletalMesh|GetAnimInstance _mesh) 0) "GroundPunch")
                      (Feedback|SpawnGenericCameraLensEffect _asfort_player_pawn (Variables|Default|GetCameraLensEffectDirtClass))
                      (if (Utilities|Array|IsValidIndex ParticleComponents)
                        (Utilities|IsValid _output
                          (:Is%20Valid
                            (Effects|Components|ParticleSystem|SetBoolParameter _output "IS_Fail" false))
                          (:Is%20Not%20Valid
                            _))
                        ;[1 0 0] VFX: Radial Force
                        (if (Math|Boolean|ANDBoolean (not (Fortnite|IsMobilePlatform)) (not (Class|PlatformHelpersLibrary|IsPcPerfMode 0)) (> (Rendering|GetRenderingDetailMode) 1))
                          (Game|SpawnActorfromClass 0 "/Game/Effects/Niagara/Blueprints/PhsyicsFields/FN_RadialForce.FN_RadialForce_C" (Math|Transform|MakeTransform (Transformation|GetWorldLocation (Variables|/Script/Engine.Actor|GetRootComponent _asfort_player_pawn))) "Undefined" "MultiplyWithRoot" 0 3.0 0.6 (Variables|Default|GetForce%20Impact%20Radius) (Variables|Default|GetForce%20Strength) 0.1 0))
                        (else
                          _))))))
              (:else))))))
    (:then_1
      (return))))
