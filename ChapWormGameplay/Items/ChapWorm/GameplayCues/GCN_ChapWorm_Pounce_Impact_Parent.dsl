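; Parent gameplay cue notify for ChapWorm pounce impacts that spawns a configurable Niagara effect on burst.
; OnBurst calls its parent, then places the effect at an adjusted impact point, oriented along the impact normal.
; GetAdjustedImpactPoint line-traces the hit component from the target pawn's location toward that component's closest collision point.
; It falls back to the effect context's original hit point and normal when that fails.
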(fn GetAdjustedImpactPoint
  (params (TargetActor :type Actor%20Object%20Reference) (Hit :type Hit%20Result%20Structure%20%28by%20ref%29))
  (bind (_bblockinghit _binitialoverlap _time _distance _location _impactpoint _normal _impactnormal _physmat _hitactor _hitcomponent) (Collision|BreakHitResult Hit))
  (bind _hitcomponent_2 (LocalVariables|GetHitComponent))
  ; AdjustedImpactPoint: Adjust the impact point so that it is closest to the bone/socket on the player that is actually doing the damage.
  (Utilities|FlowControl|Sequence
    (:then_0
      (bind _output_get (LocalVariables|SetHitComponent _hitcomponent))
      (Utilities|IsValid _output_get
        (:Is%20Valid
          (bind _asfort_player_pawn (Utilities|Casting|CastToFortPlayerPawn TargetActor))
          (bind _output_get_2 (LocalVariables|SetTraceStartLoc (Transformation|GetActorLocation _asfort_player_pawn)))
          (LocalVariables|SetClosestPointOnHitComponent (Collision|GetClosestPointonCollision _hitcomponent_2 _output_get_2 "None"))
          (if (>= (.ReturnValue (Collision|GetClosestPointonCollision _hitcomponent_2 _output_get_2 "None")) 0)
            (bind (_hitlocation _hitnormal _bonename _outhit _returnvalue) (Collision|LineTraceComponent (LocalVariables|GetHitComponent) (LocalVariables|GetTraceStartLoc) (+ (LocalVariables|GetTraceStartLoc) (* (Math|Vector|GetUnitDirection(Vector) (LocalVariables|GetTraceStartLoc) (LocalVariables|GetClosestPointOnHitComponent)) 1e+03)) true false false))
            (if _returnvalue
              (return _hitlocation _hitnormal))))))
    (:then_1
      (return _impactpoint _impactnormal))))

(fn OnBurst(EitherVFX-preferthisversion)
  (params (MyTarget :type Actor%20Object%20Reference) (Parameters :type Gameplay%20Cue%20Parameters%20Structure%20%28by%20ref%29) (ParticleComponents :type Array%20of%20FXSystem%20Component%20Object%20References) (AudioComponents :type Array%20of%20Audio%20Component%20Object%20References) (BurstCameraShakeInstance :type Camera%20Shake%20Base%20Object%20Reference) (BurstDecalInstance :type Decal%20Actor%20Object%20Reference))
  (bind (_impactpoint _impactnormal) (Default|GetAdjustedImpactPoint MyTarget (Ability|EffectContext|GetHitResult (.EffectContext (Ability|GameplayCue|BreakGameplayCueParameters Parameters)))))
  (bind (_normalizedmagnitude _rawmagnitude _effectcontext) (Ability|GameplayCue|BreakGameplayCueParameters Parameters))
  ;[0.05109 0.15 0.1264 0.5] OnBurst: Spawn the niagara effect at the desired location.
  (|CallParentFunction MyTarget Parameters ParticleComponents AudioComponents BurstCameraShakeInstance BurstDecalInstance)
  (Utilities|FlowControl|Sequence
    (:then_0
      (Utilities|IsValid MyTarget
        (:Is%20Valid
          (Niagara|SpawnSystematLocation MyTarget (Variables|Settings|GetImpactNiagaraSystem) _impactpoint (Math|Conversions|RotationFromXVector _impactnormal) "1.000000,1.000000,1.000000" true true "AutoRelease"))))
    (:then_1
      (return))))
