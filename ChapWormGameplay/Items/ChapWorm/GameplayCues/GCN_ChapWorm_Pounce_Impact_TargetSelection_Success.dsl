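; Burst gameplay cue for a successful target selection on a ChapWorm Pounce impact, playing hit VFX on the target player.
; After calling the parent burst handler, it casts the target to a Fort player pawn and gets its mesh.
; The pawn's ChapWorm.Overpower tag selects which auto-released Niagara system attaches to that mesh: the ground-punch impact hit when present, otherwise the slash impact decal effect.
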(fn OnBurst(EitherVFX-preferthisversion)
  (params (MyTarget :type Actor%20Object%20Reference) (Parameters :type Gameplay%20Cue%20Parameters%20Structure%20%28by%20ref%29) (ParticleComponents :type Array%20of%20FXSystem%20Component%20Object%20References) (AudioComponents :type Array%20of%20Audio%20Component%20Object%20References) (BurstCameraShakeInstance :type Camera%20Shake%20Base%20Object%20Reference) (BurstDecalInstance :type Decal%20Actor%20Object%20Reference))
  (|CallParentFunction MyTarget Parameters ParticleComponents AudioComponents BurstCameraShakeInstance BurstDecalInstance)
  (bind _asfort_player_pawn (Utilities|Casting|CastToFortPlayerPawn MyTarget))
  (bind _mesh (Variables|/Script/Engine.Character|GetMesh _asfort_player_pawn)
    (:then
      (if (GameplayTags|HasMatchingGameplayTag _asfort_player_pawn "(TagName=\"ChapWorm.Overpower\")")
        (Niagara|SpawnSystemAttached "/ChapWormGameplay/Items/ChapWorm/FX/NS_ChapWorm_GroundPunch_Impact_Hit.NS_ChapWorm_GroundPunch_Impact_Hit" _mesh "None" "0, 0, 0" "0, 0, 0" "KeepRelativeOffset" true true "AutoRelease")
        (else
          (Niagara|SpawnSystemAttached "/ChapWormGameplay/Items/ChapWorm/FX/NS_ChapWorm_Slash_Impact_decal.NS_ChapWorm_Slash_Impact_decal" _mesh "None" "0, 0, 0" "0, 0, 0" "KeepRelativeOffset" true true "AutoRelease")))
      (return))
    (:else)))
